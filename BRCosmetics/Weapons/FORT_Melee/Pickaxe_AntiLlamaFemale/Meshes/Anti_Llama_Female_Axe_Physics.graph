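# Physics asset for the Anti Llama Female pickaxe mesh, defining bodies and joints for a dangling knotted rope on the two-handed melee weapon.
# A large box covers the handle bone, with boxes on the two knot bones and capsules on the rope segments.
# The chain runs from handle through rope_01 and both knots, then forks into two two-segment strands.
# Twist limits are tight throughout, and the handle joint swings mainly on one axis.
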PhysicsAsset /BRCosmetics/Weapons/FORT_Melee/Pickaxe_AntiLlamaFemale/Meshes/Anti_Llama_Female_Axe_Physics
bodies {
 Melee_TwoHanded_Handle [{shapeType=Box, extentX=73.6299, extentY=69.9631, extentZ=124.6148, center=<0.4668 10.3782 41.09>, rotation=<rot -1.8894 -89.9999 -0.0415>}]
 dyn_knot [{shapeType=Box, extentX=5.544, extentY=3.9186, extentZ=5.6802, center=<-0.0527 0.0912 0.18>, rotation=<rot 4.7946 -83.5844 -13.3256>}]
 dyn_knot_01 [{shapeType=Box, extentX=3.8549, extentY=1.9689, extentZ=2.2324, rotation=<rot 0 -90 0>}]
 dyn_rope_01 [{shapeType=Capsule, length=2.2612, radius=2.1908, center=<2.2016 -1.6509 -0.041>, rotation=<rot 0 48.8611 -89.5944>}]
 dyn_rope_02 [{shapeType=Capsule, length=2.5305, radius=0.7146, center=<0.791 -0.206 0.0083>, rotation=<rot 90 0.0198 0.0199>}]
 dyn_rope_03 [{shapeType=Capsule, length=1.3119, radius=0.7181, center=<1.0976 0.0817 0.1309>, rotation=<rot 90 0 0>}]
 dyn_rope_04 [{shapeType=Capsule, length=2.6447, radius=0.7272, center=<0.9653 0.0296 0.0065>, rotation=<rot -84.9838 -179.9998 -175>}]
 dyn_rope_05 [{shapeType=Capsule, length=1.2538, radius=0.7783, center=<1.1311 -0.0402 0.2594>, rotation=<rot 90 0 0>}]
}
constraints [
 {bone1Name=dyn_rope_01, bone2Name=Melee_TwoHanded_Handle, swing1LimitDegrees=30, swing2LimitDegrees=1, twistLimitDegrees=1}
 {bone1Name=dyn_rope_02, bone2Name=dyn_knot, swing1LimitDegrees=45, swing2LimitDegrees=45, twistLimitDegrees=1}
 {bone1Name=dyn_rope_03, bone2Name=dyn_rope_02, swing1LimitDegrees=45, swing2LimitDegrees=45, twistLimitDegrees=1}
 {bone1Name=dyn_rope_04, bone2Name=dyn_knot, swing1LimitDegrees=45, swing2LimitDegrees=45, twistLimitDegrees=1}
 {bone1Name=dyn_rope_05, bone2Name=dyn_rope_04, swing1LimitDegrees=45, swing2LimitDegrees=45, twistLimitDegrees=1}
 {bone1Name=dyn_knot, bone2Name=dyn_knot_01, swing1LimitDegrees=45, swing2LimitDegrees=45, twistLimitDegrees=1}
 {bone1Name=dyn_knot_01, bone2Name=dyn_rope_01, swing1LimitDegrees=55, swing2LimitDegrees=55, twistLimitDegrees=1}
]
